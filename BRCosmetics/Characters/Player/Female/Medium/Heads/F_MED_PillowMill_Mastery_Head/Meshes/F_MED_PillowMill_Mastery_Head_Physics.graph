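# Physics asset for the PillowMill Mastery female medium head cosmetic, defining a short jointed tail chain attached to the head.
# Three small capsule bodies cover the dyn_tail bones, and a large zero-length capsule acts as a sphere around the head.
# Constraints link head, dyn_tail, dyn_tail1 and dyn_tail2 in sequence, each sharing the same moderate swing limits and tight twist limit.
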PhysicsAsset /BRCosmetics/Characters/Player/Female/Medium/Heads/F_MED_PillowMill_Mastery_Head/Meshes/F_MED_PillowMill_Mastery_Head_Physics
bodies {
 dyn_tail [{shapeType=Capsule, length=2.125, radius=1.7332, shapeName=dyn_tail_capsule, center=<0.4911 0 0>, rotation=<rot 90 0 0>}]
 dyn_tail1 [{shapeType=Capsule, length=3.4516, radius=2.0751, shapeName=dyn_tail1_capsule, center=<2.2598 0 0>, rotation=<rot 90 0 0>}]
 dyn_tail2 [{shapeType=Capsule, length=3.4889, radius=2.1138, shapeName=dyn_tail2_capsule, center=<2.3291 0 0>, rotation=<rot 90 0 0>}]
 head [{shapeType=Capsule, length=0, radius=9.6094, center=<5.1168 0.3278 -0.0067>, rotation=<rot 0 -133.6818 -88.8822>}]
}
constraints [
 {bone1Name=dyn_tail, bone2Name=head, swing1LimitDegrees=35, swing2LimitDegrees=35, twistLimitDegrees=5}
 {bone1Name=dyn_tail1, bone2Name=dyn_tail, swing1LimitDegrees=35, swing2LimitDegrees=35, twistLimitDegrees=5}
 {bone1Name=dyn_tail2, bone2Name=dyn_tail1, swing1LimitDegrees=35, swing2LimitDegrees=35, twistLimitDegrees=5}
]
