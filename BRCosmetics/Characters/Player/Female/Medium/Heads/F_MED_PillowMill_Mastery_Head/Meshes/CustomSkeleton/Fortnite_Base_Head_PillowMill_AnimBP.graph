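# Animation Blueprint for the PillowMill Mastery female cosmetic head, adding physics-driven secondary motion on its custom head skeleton.
# The input pose passes through one RigidBody node, simulated relative to the head bone with the head's physics asset and overridden gravity, then goes to the output.
# A dyn_tail-to-head Constraint and a second, differently tuned RigidBody node sit unconnected.
# The event graph holds only unwired stub nodes.
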AnimBlueprint /BRCosmetics/Characters/Player/Female/Medium/Heads/F_MED_PillowMill_Mastery_Head/Meshes/CustomSkeleton/Fortnite_Base_Head_PillowMill_AnimBP
target_skeleton=&/Fortnite_Base_Head_PillowMil_Skeleton
graphs [
 AnimGraph {nodes [
   "Constraint - dyn_tail (P)"(Alpha=1, ConstraintWeights_0=1) {
    alphaBoolBlend {blendOption=Linear}
    alphaScaleBiasClamp {scale=0.5}
    boneToModify {boneName=dyn_tail}
    constraintSetup [{targetBone {boneName=head}}]
    constraintWeights [1]
   }
   RigidBody!(ComponentPose, Alpha, OverrideWorldGravity:Vector="0,0,-980", ExternalForce) -> (Pose) {
    bClampLinearTranslationLimitToRefPose=true
    bOverrideWorldGravity=true
    overridePhysicsAsset=/BRCosmetics/Characters/Player/Female/Medium/Heads/F_MED_PillowMill_Mastery_Head/Meshes/F_MED_PillowMill_Mastery_Head_Physics
    simulationSpace=BaseBoneSpace
    alphaBoolBlend {blendOption=Linear}
    alphaScaleBiasClamp {bInterpResult=true, interpSpeedDecreasing=5, interpSpeedIncreasing=5}
    baseBoneRef {boneName=head}
    componentAppliedLinearAccClamp=<3000 3000 3000>
    componentLinearAccScale=<1 1 0.6>
    componentLinearVelScale=<0.6 0.6 0.6>
    overrideWorldGravity {z=-980}
    simSpaceSettings {maxAngularAcceleration=500, maxAngularVelocity=150, maxLinearAcceleration=400, maxLinearVelocity=250, velocityScaleZ=0, worldAlpha=0.38}
   }
   n1 = RigidBody!(ComponentPose=Local To Component(LocalPose=Input Pose()), Alpha, ExternalForce) -> (Pose) {
    bClampLinearTranslationLimitToRefPose=true
    bOverrideWorldGravity=true
    overridePhysicsAsset=&F_MED_PillowMill_Mastery_Head_Physics
    simulationSpace=BaseBoneSpace
    alphaBoolBlend {blendOption=Linear}
    baseBoneRef {boneName=head}
    componentAppliedLinearAccClamp=<150 150 50>
    componentLinearAccScale=<0.3 0.3 0.3>
    componentLinearVelScale=<0.3 0.3 0.3>
    overrideWorldGravity {z=-980}
    simSpaceSettings {maxAngularAcceleration=1500, maxLinearAcceleration=1500, worldAlpha=1}
   }
   Output Pose(Result=Component To Local(ComponentPose=@n1))
  ]}
 EventGraph {nodes [
   TryGetPawnOwner()
   Event BlueprintUpdateAnimation()
  ]}
]
variables []
